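# Animation Blueprint for the M_LRG_ZebraScramble_Bacon_Head cosmetic head of large male player characters, targeting that head's own skeleton.
# Each animation update calls the parent implementation, then copies the R_smile_pose curve value into the NewVar variable.
# Nothing in the shown graph reads NewVar, and a TryGetPawnOwner node sits unconnected.
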AnimBlueprint /BRCosmetics/Characters/Player/Male/Large/Heads/M_LRG_ZebraScramble_Bacon_Head/Meshes/M_LRG_ZebraScramble_Bacon_Head_AnimBP
target_skeleton=/Game/Characters/Player/Male/Large/Heads/M_LRG_ZebraScramble_Bacon_Head/Meshes/M_LRG_ZebraScramble_Bacon_Head_Skeleton
graphs [EventGraph {nodes [
   n1 = Event BlueprintUpdateAnimation()
   > Parent: BlueprintUpdateAnimation!(execute, DeltaTimeX:float=@n1.DeltaTimeX) -> (then) {is=title K2Node_CallParentFunction}
   > Set NewVar(NewVar:float=GetCurveValue(CurveName=R_smile_pose)) -> (Output_Get:float)
   TryGetPawnOwner()
  ]}]
variables [NewVar {type=number}]
