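# Post-process AnimBlueprint for the ZebraScramble Bacon head that runs facial Control Rigs and corrects teeth and mouth shapes.
# An NPC meta tag, checked on customization load, selects the NPC or player rig; both receive the emote state.
# A Modify Bone chain then applies constant teeth offsets and two curve-weighted fixes.
# The emote fix reshapes teeth, jaw and lips, scaled by jaw opening; the NPC-elimination fix adjusts only teeth.
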AnimBlueprint /BRCosmetics/Characters/Player/Male/Large/Heads/M_LRG_ZebraScramble_Bacon_Head/Meshes/M_LRG_ZebraScramble_Bacon_Head_Skeleton_PPAnimBlueprint
target_skeleton=/Game/Characters/Player/Male/Large/Heads/M_LRG_ZebraScramble_Bacon_Head/Meshes/M_LRG_ZebraScramble_Bacon_Head_Skeleton
graphs [
 AnimGraph {nodes [
   AnimGraphNode_LinkedInputPose_0 = Input Pose() {binding=&self:AnimGraph.AnimGraphNode_LinkedInputPose_0.Binding}
   n1 = Get isNPC() -> (isNPC:bool)
   n2 = Get bIsPlayingEmote()
   AnimGraphNode_ControlRig_1 = Control Rig!(Source=@AnimGraphNode_LinkedInputPose_0, bAlphaBoolEnabled:bool=NOT Boolean(A=@n1), isEmoting:bool=@n2) -> (Pose) {
    binding=&self:AnimGraph.AnimGraphNode_ControlRig_1.Binding
    alphaInputType=Bool
    alphaBoolBlend {blendOption=Linear}
    controlRigAssetReference {blueprintRigClass=&/M_LRG_ZebraScramble_Bacon_Head_CtrlRig.M_LRG_ZebraScramble_Bacon_Head_CtrlRig_C}
   }
   AnimGraphNode_ControlRig_0 = Control Rig(Source=@AnimGraphNode_ControlRig_1, bAlphaBoolEnabled=@n1, isEmoting=@n2) {
    binding=&self:AnimGraph.AnimGraphNode_ControlRig_0.Binding
    alphaInputType=Bool
    alphaBoolBlend {blendOption=Linear}
    controlRigAssetReference {blueprintRigClass=&/M_LRG_ZebraScramble_Bacon_Head_CtrlRig_NPC.M_LRG_ZebraScramble_Bacon_Head_CtrlRig_NPC_C}
   }
   n3 = "Transform (Modify) Bone - Bone: teeth_upper"(
    ComponentPose=Local To Component(LocalPose=@AnimGraphNode_ControlRig_0)
    Alpha=1
    Translation="0,1.5,-1.471"
    Rotation="0,0,0"
    Scale="1.2,1,1"
   ) {
    rotationMode=BMM_Additive
    rotationSpace=BCS_ParentBoneSpace
    scaleMode=BMM_Replace
    scaleSpace=BCS_ParentBoneSpace
    translationMode=BMM_Additive
    translationSpace=BCS_ParentBoneSpace
    alphaBoolBlend {blendOption=Linear}
    boneToModify {boneName=teeth_upper}
    scale {x=1.2}
    translation {y=1.5, z=-1.471}
   }
   n4 = "Transform (Modify) Bone - Bone: teeth_lower"(ComponentPose=@n3, Alpha=1, Translation="0,1.5,0.5", Rotation="0,0,0", Scale="1.2,1,1") {
    rotationMode=BMM_Additive
    rotationSpace=BCS_ParentBoneSpace
    scaleMode=BMM_Replace
    scaleSpace=BCS_ParentBoneSpace
    translationMode=BMM_Additive
    translationSpace=BCS_ParentBoneSpace
    alphaBoolBlend {blendOption=Linear}
    boneToModify {boneName=teeth_lower}
    scale {x=1.2}
    translation {y=1.5, z=0.5}
   }
   n5 = Get emote_teeth_fix() -> (emote_teeth_fix:float)
   n6 = "Transform (Modify) Bone - Bone: teeth_upper"(ComponentPose=@n4, Alpha=@n5, Translation="0,0,-1.3", Rotation="0,0,0", Scale="1.5,1,1.8") {
    rotationMode=BMM_Additive
    rotationSpace=BCS_ParentBoneSpace
    scaleMode=BMM_Replace
    scaleSpace=BCS_ParentBoneSpace
    translationMode=BMM_Additive
    translationSpace=BCS_ParentBoneSpace
    alphaBoolBlend {blendOption=Linear}
    boneToModify {boneName=teeth_upper}
    scale {x=1.5, z=1.8}
    translation {z=-1.3}
   }
   n7 = "Transform (Modify) Bone - Bone: teeth_lower"(ComponentPose=@n6, Alpha=@n5, Translation="0,0,3.8", Rotation="0,0,-15", Scale="1.55,1,3") {
    rotationMode=BMM_Additive
    rotationSpace=BCS_ParentBoneSpace
    scaleMode=BMM_Replace
    scaleSpace=BCS_ParentBoneSpace
    translationMode=BMM_Additive
    translationSpace=BCS_ParentBoneSpace
    alphaBoolBlend {blendOption=Linear}
    boneToModify {boneName=teeth_lower}
    rotation {roll=-15}
    scale {x=1.55, z=3}
    translation {z=3.8}
   }
   n8 = "Transform (Modify) Bone - Bone: C_jaw"(ComponentPose=@n7, Alpha=@n5, Translation="0,0,-5", Rotation="0,0,0", Scale="1,1,1") {
    rotationMode=BMM_Additive
    rotationSpace=BCS_ParentBoneSpace
    scaleMode=BMM_Replace
    scaleSpace=BCS_ParentBoneSpace
    translationMode=BMM_Additive
    translationSpace=BCS_ParentBoneSpace
    alphaBoolBlend {blendOption=Linear}
    boneToModify {boneName=C_jaw}
    translation {z=-5}
   }
   n9 = "Transform (Modify) Bone - Bone: C_lip_lower_mid"(ComponentPose=@n8, Alpha=@n5, Translation="0,0,5", Rotation="0,0,0", Scale="2,1,1") {
    rotationMode=BMM_Additive
    rotationSpace=BCS_ParentBoneSpace
    scaleMode=BMM_Replace
    scaleSpace=BCS_ParentBoneSpace
    translationMode=BMM_Additive
    translationSpace=BCS_ParentBoneSpace
    alphaBoolBlend {blendOption=Linear}
    boneToModify {boneName=C_lip_lower_mid}
    scale {x=2}
    translation {z=5}
   }
   n10 = "Transform (Modify) Bone - Bone: L_lip_lower_outer"(ComponentPose=@n9, Alpha=@n5, Translation="1,0,-4", Rotation="0,0,0", Scale="1,1,1") {
    rotationMode=BMM_Additive
    rotationSpace=BCS_ParentBoneSpace
    scaleMode=BMM_Replace
    scaleSpace=BCS_ParentBoneSpace
    translationMode=BMM_Additive
    translationSpace=BCS_ParentBoneSpace
    alphaBoolBlend {blendOption=Linear}
    boneToModify {boneName=L_lip_lower_outer}
    translation {x=1, z=-4}
   }
   n11 = "Transform (Modify) Bone - Bone: R_lip_lower_outer"(ComponentPose=@n10, Alpha=@n5, Translation="-1,0,-4", Rotation="0,0,0", Scale="1,1,1") {
    rotationMode=BMM_Additive
    rotationSpace=BCS_ParentBoneSpace
    scaleMode=BMM_Replace
    scaleSpace=BCS_ParentBoneSpace
    translationMode=BMM_Additive
    translationSpace=BCS_ParentBoneSpace
    alphaBoolBlend {blendOption=Linear}
    boneToModify {boneName=R_lip_lower_outer}
    translation {x=-1, z=-4}
   }
   n12 = "Transform (Modify) Bone - Bone: L_lip_corner"(ComponentPose=@n11, Alpha=@n5, Translation="2,0,5", Rotation="0,0,45", Scale="1,1,1") {
    rotationMode=BMM_Additive
    rotationSpace=BCS_ParentBoneSpace
    scaleMode=BMM_Replace
    scaleSpace=BCS_ParentBoneSpace
    translationMode=BMM_Additive
    translationSpace=BCS_ParentBoneSpace
    alphaBoolBlend {blendOption=Linear}
    boneToModify {boneName=L_lip_corner}
    rotation {roll=45}
    translation {x=2, z=5}
   }
   n13 = "Transform (Modify) Bone - Bone: R_lip_corner"(ComponentPose=@n12, Alpha=@n5, Translation="-2,0,5", Rotation="0,0,45", Scale="1,1,1") {
    rotationMode=BMM_Additive
    rotationSpace=BCS_ParentBoneSpace
    scaleMode=BMM_Replace
    scaleSpace=BCS_ParentBoneSpace
    translationMode=BMM_Additive
    translationSpace=BCS_ParentBoneSpace
    alphaBoolBlend {blendOption=Linear}
    boneToModify {boneName=R_lip_corner}
    rotation {roll=45}
    translation {x=-2, z=5}
   }
   n14 = "Transform (Modify) Bone - Bone: L_lip_upper_outer"(ComponentPose=@n13, Alpha=@n5, Translation="0,0,-2", Rotation="0,0,0", Scale="1,1,1") {
    rotationMode=BMM_Additive
    rotationSpace=BCS_ParentBoneSpace
    scaleMode=BMM_Replace
    scaleSpace=BCS_ParentBoneSpace
    translationMode=BMM_Additive
    translationSpace=BCS_ParentBoneSpace
    alphaBoolBlend {blendOption=Linear}
    boneToModify {boneName=L_lip_upper_outer}
    translation {z=-2}
   }
   n15 = "Transform (Modify) Bone - Bone: R_lip_upper_outer"(ComponentPose=@n14, Alpha=@n5, Translation="0,0,-2", Rotation="0,0,0", Scale="1,1,1") {
    rotationMode=BMM_Additive
    rotationSpace=BCS_ParentBoneSpace
    scaleMode=BMM_Replace
    scaleSpace=BCS_ParentBoneSpace
    translationMode=BMM_Additive
    translationSpace=BCS_ParentBoneSpace
    alphaBoolBlend {blendOption=Linear}
    boneToModify {boneName=R_lip_upper_outer}
    translation {z=-2}
   }
   n16 = Get npc_elim_teeth_fix() -> (npc_elim_teeth_fix:float)
   n17 = "Transform (Modify) Bone - Bone: teeth_upper"(ComponentPose=@n15, Alpha=@n16, Translation="0,0.8,0", Rotation="0,0,5", Scale="1.5,1,1") {
    rotationMode=BMM_Additive
    rotationSpace=BCS_ParentBoneSpace
    scaleMode=BMM_Replace
    scaleSpace=BCS_ParentBoneSpace
    translationMode=BMM_Additive
    translationSpace=BCS_ParentBoneSpace
    alphaBoolBlend {blendOption=Linear}
    boneToModify {boneName=teeth_upper}
    rotation {roll=5}
    scale {x=1.5}
    translation {y=0.8}
   }
   n18 = "Transform (Modify) Bone - Bone: teeth_lower"(ComponentPose=@n17, Alpha=@n16, Translation="0,0.8,0", Rotation="0,0,5", Scale="1.5,1,1") {
    rotationMode=BMM_Additive
    rotationSpace=BCS_ParentBoneSpace
    scaleMode=BMM_Replace
    scaleSpace=BCS_ParentBoneSpace
    translationMode=BMM_Additive
    translationSpace=BCS_ParentBoneSpace
    alphaBoolBlend {blendOption=Linear}
    boneToModify {boneName=teeth_lower}
    rotation {roll=5}
    scale {x=1.5}
    translation {y=0.8}
   }
   AnimGraphNode_Root_0 = Output Pose(Result=Component To Local(ComponentPose=@n18)) {
    binding=&self:AnimGraph.AnimGraphNode_Root_0.Binding
    layerGroup=None
   }
  ]}
 EventGraph {nodes [
   Event BlueprintUpdateAnimation()
   n1 = GetCurveValue(CurveName=jaw_open_pose)
   > Set emote_teeth_fix(emote_teeth_fix:float=float * float(A=GetCurveValue(CurveName=ZebraScramble_Emote_TeethFix), B=MapRangeClamped(Value=@n1, InRangeA=0, InRangeB=0.626)) {is=title K2Node_PromotableOperator}) -> (Output_Get:float)
   > Set npc_elim_teeth_fix(npc_elim_teeth_fix:float=GetCurveValue(CurveName=ZebraScramble_NPCElim_TeethFix)) -> (Output_Get:float)
   n2 = Event OnCharacterCustomizationLoaded()
   > Set isNPC(isNPC:bool=CharacterItemDefinitionHasAnyMetaTag(Tags=Make Array("[0]"=(TagName="Cosmetics.MetaTags.NPC")))) -> (Output_Get:bool)
   n3 = Branch(Condition="Equal (Object)"(
     A=Get Array Item with EFort Custom Part Type(Array:Array of Custom Character Part Object References=@n2.CustomParts) -> (Output:Custom Character Part Object Reference)
     B=/BRCosmetics/Athena/Heroes/Meshes/Bodies/CP_Athena_Body_M_ZebraScramble_NPC
    ))
   > Set isplayingemoteorNPC(isplayingemoteorNPC:bool=true) -> (Output_Get:bool)
   Set isplayingemoteorNPC(execute=@n3.else, isplayingemoteorNPC=Get bIsPlayingEmote())
   TryGetPawnOwner()
  ]}
]
variables [
 emote_teeth_fix {type=number}
 isNPC {type=boolean}
 isplayingemoteorNPC {type=boolean}
 npc_elim_teeth_fix {type=number}
]
